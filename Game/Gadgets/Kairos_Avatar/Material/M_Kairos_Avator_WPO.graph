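# Material function for the Kairos avatar gadget that outputs a World Position Offset combining a floating up-down bob with a slight yaw sway.
# The RandomNumberFrac input varies float and rotation speeds between base and base-plus-add parameters.
# Each speed multiplies the RandomRime input into a sine that scales its maximum movement parameter.
# The results drive a Z offset and a RotateAboutAxis around world Z, pivoted at object position, which are summed.
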MaterialFunction /Game/Gadgets/Kairos_Avatar/Material/M_Kairos_Avator_WPO
graph {nodes [
  n1 = Constant2Vector()
  n2 = ScalarParameter() {defaultValue=0.08, parameterName=FloatingSpeed}
  n3 = FunctionInput() {bUsePreviewValueAsDefault=true, inputName=RandomNumberFrac, inputType=FunctionInput_Scalar}
  n4 = LinearInterpolate(A=@n2, B=Add(A=@n2, B=ScalarParameter() {defaultValue=0.02, parameterName=FloatingSpeedadd}), Alpha=@n3)
  n5 = FunctionInput() {bUsePreviewValueAsDefault=true, inputName=RandomRime, inputType=FunctionInput_Scalar}
  n6 = LinearInterpolate(B=ScalarParameter() {defaultValue=2.6296, parameterName=UpdownMovement}, Alpha=Sine(None=Multiply(A=@n5, B=@n4)))
  n7 = Constant3Vector() {constant=<rgba 0 0 1 0>}
  n8 = ScalarParameter() {defaultValue=0.08, parameterName=RotationSpeed}
  n9 = LinearInterpolate(A=@n8, B=Add(A=@n8, B=ScalarParameter() {defaultValue=0.01, parameterName=RoationSpeedadd}), Alpha=@n3)
  n10 = LinearInterpolate(B=ScalarParameter() {defaultValue=0.01, parameterName=maxRotationMovement}, Alpha=Sine(None=Multiply(A=@n5, B=@n9)))
  n11 = ObjectPositionWS()
  n12 = WorldPosition()
  FunctionOutput(None=Add(A=AppendVector(A=@n1, B=@n6), B=RotateAboutAxis(NormalizedRotationAxis=@n7, RotationAngle=@n10, PivotPoint=@n11, Position=@n12.XYZ))) {outputName=WPO}
 ]}
