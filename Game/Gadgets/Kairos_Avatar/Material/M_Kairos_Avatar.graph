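# Material for the Kairos Avatar gadget: an animated screen that blends between a loading spinner and a character portrait parameter.
# The background layers panning square patterns, a pulsing vignette and scanlines, while glitch masks offset the portrait's UVs.
# Vertex colors and pivots read from extra UV channels drive oscillating square displacement, added to a WPO function.
# Object position randomizes timing; quality and shading-path switches drop most effects on Low, Forward and Mobile.
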Material /Game/Gadgets/Kairos_Avatar/Material/M_Kairos_Avatar
graph {
 properties {pixelDepthOffsetMode=PDOM_Legacy, usedShadingModels=Default Lit}
 nodes [
  n1 = MaterialFunctionCall(Float3=ObjectPositionWS()) {materialFunction=&BreakOutFloat3Components}
  n2 = Frac(None=Divide(A=Multiply(A=@n1.R, B=@n1.R), B=@n1.B))
  n3 = Multiply(A=@n2, B=ScalarParameter() {defaultValue=5347, parameterName=random})
  n4 = Add(A=@n3, B=Time())
  n5 = VectorParameter() {parameterName=SquaresScaleMultiplier, defaultValue=<rgba 1 0 0.4 0>}
  n6 = VertexColor()
  n7 = MaterialFunctionCall(Float2=TextureCoordinate() {coordinateIndex=2}) {materialFunction=&BreakOutFloat2Components}
  n8 = MaterialFunctionCall(
   X=@n7.R
   Y=@n7.R
   Z=MaterialFunctionCall(Float2=TextureCoordinate() {coordinateIndex=3}) {materialFunction=&BreakOutFloat2Components}.R
  ) {materialFunction=&MakeFloat3}
  n9 = MaterialFunctionCall(Float3=Transform(None=@n8.Result) {transformSourceType=TRANSFORMSOURCE_Local}) {materialFunction=&BreakOutFloat3Components}
  n10 = Multiply(A=@n9.R, B=@n9.R)
  n11 = Add(A=@n9.B) {constB=0.0001}
  n12 = Multiply(A=Frac(None=Divide(A=@n10, B=@n11)), B=ScalarParameter() {defaultValue=5347, parameterName=random})
  n13 = Time()
  n14 = Sine(None=Multiply(A=Add(A=@n12, B=@n13), B=ScalarParameter() {defaultValue=0.5, parameterName=Speed}))
  n15 = Subtract(
   A=MaterialFunctionCall() {materialFunction=&LocalPosition}.Local Position
   B=@n8.Result
  )
  Add(
   A=MaterialFunctionCall!(RandomRime=@n4, RandomNumberFrac=@n2) {materialFunction=&/M_Kairos_Avator_WPO}.WPO
   B=Multiply(A=Multiply(A=@n5.RGB, B=@n6), B=Multiply(A=@n14, B=@n15))
  )
  n16 = ComponentMask(None=VertexInterpolator(VS=PreSkinnedNormal()).PS) {g=true}
  n17 = VertexColor()
  n18 = Saturate(None=Add(A=OneMinus(None=@n16), B=Multiply(A=@n17.G) {constB=0.8}))
  n19 = VectorParameter() {parameterName=BackgroundRGB, defaultValue=<rgba 0 0.095 0.3906 1>}
  n20 = Power(Base=@n19.RGB) {constExponent=0.75}
  n21 = Multiply(A=Subtract(A=Multiply(A=@n18, B=@n20)) {constB=0.01}) {constB=2}
  n22 = TextureCoordinate()
  n23 = Time()
  n24 = Multiply(
   A=Panner(Coordinate=Multiply(A=@n22) {constB=0.75}, Time=Multiply(A=@n23) {constB=0.5}) {speedX=-0.05, speedY=0.02}
   B=AppendVector(A=ScalarParameter() {defaultValue=1.3059, parameterName=UVscale_x}, B=ScalarParameter() {defaultValue=0.9895, parameterName=UVscale_y})
  )
  n25 = ScalarParameter() {defaultValue=-0.14, parameterName=LoadingUVoffset_x}
  n26 = ScalarParameter() {parameterName=LoadingUVoffset_y}
  n27 = Add(A=@n24, B=AppendVector(A=Add(A=@n25) {constB=-0.33}, B=Add(A=@n26) {constB=-0.1}))
  n28 = Panner(Coordinate=Saturate(None=Frac(None=@n27)), Time=@n4) {speedX=-0.025, speedY=0.15}
  n29 = Constant() {r=0.125}
  n30 = Multiply(A=@n4) {constB=0.15}
  n31 = Add(A=Multiply(A=Sine(None=@n30)) {constB=0.2}) {constB=0.8}
  n32 = Multiply(A=@n4) {constB=0.2}
  n33 = Add(A=Multiply(A=Sine(None=@n32)) {constB=0.4}) {constB=0.6}
  n34 = Add(A=@n25) {constB=-0.25}
  n35 = Add(B=@n26) {constB=-0.25}
  n36 = Frac(None=Add(A=@n24, B=AppendVector(A=@n34, B=@n35)))
  n37 = Add(A=@n4) {constB=0.25}
  n38 = Multiply(A=@n37) {constB=0.33}
  n39 = Add(A=Multiply(A=Sine(None=@n38)) {constB=0.3}) {constB=0.7}
  n40 = MaterialFunctionCall!(
   UVs=Panner(Coordinate=Saturate(None=@n36), Time=@n37) {bFractionalPart=true, speedX=0.0125, speedY=0.12}
   Size=Constant() {r=0.15}
   SizeScale=@n39
  ) {materialFunction=/Game/Packages/Fortress_SharedMaterials/Base_Material_Functions/SquarePatternFromUVs}
  n41 = Add(A=@n25) {constB=0}
  n42 = Add(A=@n26) {constB=0}
  n43 = Frac(None=Add(A=@n24, B=AppendVector(A=@n41, B=@n42)))
  n44 = MaterialFunctionCall(UVs=Panner(Coordinate=Saturate(None=@n43), Time=@n23) {speedX=-0.04, speedY=-0.01}, Size=Constant() {r=0.175}, SizeScale=@n33) {materialFunction=&SquarePatternFromUVs}
  n45 = Add(
   A=Multiply(
    A=MaterialFunctionCall(UVs=@n28, Size=@n29, SizeScale=@n31) {materialFunction=&SquarePatternFromUVs}.PatternOutput
    B=@n33
   )
   B=Add(A=Multiply(A=@n40.PatternOutput, B=@n31), B=Multiply(A=@n44.PatternOutput, B=@n39))
  )
  n46 = Add(A=Saturate(None=Multiply(A=@n45) {constB=0.1}), B=@n19.RGB)
  n47 = TextureCoordinate()
  n48 = Constant2Vector() {g=0.5, r=0.5}
  n49 = Sine(None=Multiply(A=@n4, B=ScalarParameter() {defaultValue=0.44, parameterName=SpeedFlicker}))
  n50 = LinearInterpolate(
   A=ScalarParameter() {defaultValue=0.8, parameterName=Radius}
   B=ScalarParameter() {defaultValue=0.9, parameterName=Radius2}
   Alpha=Divide(A=Add(A=@n49))
  )
  n51 = ScalarParameter() {parameterName=Hardness}
  n52 = Saturate(None=Add(A=SphereMask(A=@n47, B=@n48, Radius=@n50, Hardness=@n51), B=ScalarParameter() {defaultValue=0.4, parameterName=Brightness}))
  n53 = Multiply(A=QualitySwitch(Default=ShadingPathSwitch(Default=@n46, Deferred=@n46, Forward=@n19.RGB, Mobile=@n19.RGB), Low=@n19.RGB), B=@n52)
  n54 = TextureCoordinate()
  n55 = Add(A=@n3, B=Time())
  n56 = Panner(Coordinate=@n54, Time=@n55) {speedY=-0.025}
  n57 = Multiply(A=ComponentMask(None=@n56) {g=true}) {constB=24}
  n58 = Add(A=Multiply(A=OneMinus(None=@n52)) {constB=0.5}, B=Add(A=Frac(None=@n57)) {constB=0.8})
  n59 = Constant() {r=1.33}
  n60 = Multiply(A=Sine(None=@n55), B=Sine(None=Multiply(A=@n55, B=@n59)))
  n61 = Panner(Coordinate=@n54, Time=@n55) {speedY=0.55}
  n62 = Constant() {r=2.5}
  n63 = ComponentMask(None=Frac(None=Multiply(A=@n61, B=@n62))) {g=true}
  n64 = Constant() {r=2.4}
  n65 = ComponentMask(None=Frac(None=Multiply(A=@n56, B=@n64))) {g=true}
  n66 = Multiply(A=Add(A=@n63, B=@n65)) {constB=0.5}
  n67 = Saturate(None=Ceil(None=Subtract(A=@n66) {constB=0.6}))
  n68 = Round(None=Multiply(A=@n55) {constB=1.1})
  n69 = Frac(None=Multiply(A=Panner(Coordinate=@n54, Time=@n68) {speedY=-0.05}, B=Constant() {r=4}))
  n70 = Saturate(None=Ceil(None=Subtract(A=@n69) {constB=0.5}))
  n71 = Multiply(A=Multiply(A=@n67, B=Saturate(None=@n60)), B=Saturate(None=ComponentMask(None=@n70) {g=true}))
  n72 = Add(A=Multiply(A=@n53, B=Saturate(None=@n58)), B=Multiply(A=@n52, B=Multiply(A=@n71) {constB=0.0225}))
  n73 = QualitySwitch(Default=ShadingPathSwitch(Default=@n72, Deferred=@n72, Forward=@n53, Mobile=@n53), Low=@n53)
  n74 = AppendVector(A=ScalarParameter() {defaultValue=1.3059, parameterName=UVscale_x}, B=ScalarParameter() {defaultValue=0.9895, parameterName=UVscale_y})
  n75 = TextureCoordinate()
  n76 = ScalarParameter() {defaultValue=-0.14, parameterName=LoadingUVoffset_x}
  n77 = ScalarParameter() {parameterName=LoadingUVoffset_y}
  n78 = Rotator(
   Coordinate=Add(A=Multiply(A=@n74, B=@n75), B=AppendVector(A=@n76, B=@n77))
   Time=Multiply(A=@n4, B=ScalarParameter() {defaultValue=-15, parameterName=RotationSpeed})
  )
  n79 = LinearInterpolate(
   A=@n73
   B=Add(A=TextureSample(UVs=@n78) {texture=/Game/Athena/Items/EnvironmentalItems/SCMachine/FX/T_Loading_Spinner}.R, B=@n73) {desc=LoadingScreen}
   Alpha=ScalarParameter() {defaultValue=1, parameterName=LoadingScreen}
  )
  n80 = ScalarParameter() {defaultValue=1.15, parameterName=UVscaleImage_x}
  n81 = ScalarParameter() {defaultValue=0.8625, parameterName=UVscaleImage_y}
  n82 = Add(
   A=Multiply(A=@n75, B=AppendVector(A=@n80, B=@n81))
   B=AppendVector(A=ScalarParameter() {defaultValue=-0.03, parameterName=UVoffset_x}, B=ScalarParameter() {defaultValue=0.015, parameterName=UVoffset_y})
  )
  n83 = Multiply(A=@n82, B=Subtract(B=Multiply(A=@n14) {constB=0.01}))
  n84 = QualitySwitch(Default=ShadingPathSwitch(Default=@n83, Deferred=@n83, Forward=@n82, Mobile=@n82), Low=@n82)
  n85 = Multiply(A=@n67, B=ScalarParameter() {defaultValue=1.25, desc=UVoffsetGlitch_x, parameterName=UVscaleGlitch_x})
  n86 = Multiply(A=Subtract(A=MaterialFunctionCall(X=@n63, Y=@n65) {materialFunction=&MakeFloat2}.Result) {constB=0.5}) {constB=5}
  n87 = Saturate(None=Multiply(A=Ceil(None=@n86)) {constB=0.2})
  n88 = MaterialFunctionCall(Float2=Subtract(A=Multiply(A=@n87) {constB=2})) {materialFunction=&BreakOutFloat2Components}
  n89 = Constant() {r=0.1}
  n90 = Add(A=Multiply(A=@n88.R, B=Multiply(A=@n89) {constB=-0.75}) {constB=-0.05}, B=Multiply(A=@n89, B=@n88.G) {constB=0.05})
  n91 = AppendVector(A=Subtract(A=ComponentMask(None=@n84) {r=true}, B=Multiply(A=@n85, B=@n90)) {constB=-0.1}, B=ComponentMask(None=@n84) {g=true})
  n92 = TextureSampleParameter2D(UVs=@n91) {parameterName=LargePreview, texture=/Game/UI/Foundation/Textures/Icons/Heroes/Athena/Soldier/T-Soldier-HID-005-Athena-Commando-M-L}
  n93 = TextureSampleParameter2D(UVs=@n91) {parameterName=SmallPreview, texture=/Game/UI/Foundation/Textures/Icons/Heroes/Athena/Soldier/T-Soldier-HID-005-Athena-Commando-M}
  n94 = Multiply(A=@n71) {constB=0.8}
  n95 = Saturate(None=Multiply(A=StaticSwitchParameter(True=@n92.A, False=@n93.A) {parameterName=UseLargePreviewImage}, B=@n94))
  n96 = TextureSampleParameter2D(UVs=@n84) {parameterName=LargePreview, texture=&T-Soldier-HID-005-Athena-Commando-M-L}
  n97 = TextureSampleParameter2D(UVs=@n84) {parameterName=SmallPreview, texture=&T-Soldier-HID-005-Athena-Commando-M}
  n98 = StaticSwitchParameter(True=@n96.A, False=@n97.A) {parameterName=UseLargePreviewImage}
  n99 = StaticSwitchParameter(True=@n96.RGB, False=@n97.RGB) {parameterName=UseLargePreviewImage}
  n100 = LinearInterpolate(
   A=@n99
   B=StaticSwitchParameter(True=@n92.RGB, False=@n93.RGB) {parameterName=UseLargePreviewImage}
   Alpha=Saturate(None=Subtract(A=@n95, B=@n98))
  )
  n101 = Saturate(None=Add(A=@n98, B=@n95))
  n102 = ShadingPathSwitch(Default=@n101, Deferred=@n101, Forward=@n98, Mobile=@n98)
  n103 = VertexColor()
  n104 = LinearInterpolate(
   A=@n73
   B=QualitySwitch(Default=ShadingPathSwitch(Default=@n100, Deferred=@n100, Forward=@n99, Mobile=@n99), Low=@n99)
   Alpha=Multiply(A=QualitySwitch(Default=@n102, Low=@n98), B=OneMinus(None=@n103.G))
  ) {desc=Character}
  n105 = ScalarParameter() {parameterName=CharacterScreen}
  n106 = Multiply(A=OneMinus(None=ComponentMask(None=@n8.Result) {b=true}), B=ScalarParameter() {defaultValue=0.06, parameterName=SquaresGradientScale}) {constB=0.01}
  n107 = ScalarParameter() {defaultValue=1.2, parameterName=SquaresGradientOffset}
  LinearInterpolate(
   A=Add(A=Saturate(None=@n21), B=LinearInterpolate(A=@n79, B=@n104, Alpha=@n105))
   B=VectorParameter() {parameterName=SquaresBackgroundColor, defaultValue=<rgba 0.0785 0.2191 0.45 1>}.RGB
   Alpha=Saturate(None=Subtract(A=@n106, B=@n107))
  )
  n108 = ScalarParameter() {defaultValue=0.015, parameterName=Specular}
  n109 = Saturate(None=@n45)
  n110 = Saturate(None=Add(A=@n108, B=Multiply(A=@n109) {constB=0.05}))
  QualitySwitch(Default=ShadingPathSwitch(Default=@n110, Deferred=@n110, Forward=@n108, Mobile=@n108), Low=@n108)
  Reroute(None=VectorParameter() {parameterName=Color}.RGB)
  n111 = Subtract(A=@n98, B=@n94)
  Saturate(None=Add(A=@n95, B=Saturate(None=@n111)))
  n112 = Frac(None=Multiply(A=@n55) {constB=0.8})
  n113 = Frac(None=Multiply(A=Add(A=@n55) {constB=0.42}) {constB=1.25})
  Saturate(None=Round(None=Multiply(A=@n112, B=@n113) {constB=1.25}))
  ScalarParameter() {defaultValue=0.28, parameterName=Roughness}
  ScalarParameter() {parameterName=Metalness}
 ]
}
